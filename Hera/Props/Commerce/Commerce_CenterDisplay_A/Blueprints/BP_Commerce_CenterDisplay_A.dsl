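; Commerce center display prop Blueprint that can tint its secondary meshes from building texture data during construction.
; After the parent construction script, an optional flag gates a pass over the direct children of the secondary mesh component.
; Each child, cast to a static mesh component, is colored with the tint from texture data index 1, though the guard only checks for a non-empty array.
; The runtime events are empty.
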(event EventBeginPlay)

(event Collision|EventActorBeginOverlap
  (params (OtherActor :type Actor%20Object%20Reference)))

(event EventTick
  (params (DeltaSeconds :type Float%20%28single-precision%29)))

(fn ConstructionScript
  (|CallParentFunction)
  (if (Variables|Default|GetUse%20TD%20for%20Tint)
    (bind _outtexturedataarray (Building|GetBuildingTextureData))
    (if (> (Utilities|Array|Length _outtexturedataarray) 0)
      (for _array_element (Components|GetChildrenComponents (Variables|Default|GetSecondary%20mesh) false)
        (bind _asstatic_mesh_component (Utilities|Casting|CastToStaticMeshComponent _array_element))
        (Default|SetColoronSecondaryMesh (Math|Conversions|ToLinearColor(Color) (Variables|/Script/FortniteGame.BuildingTextureData|GetTintColor (Utilities|Array|Get(acopy) _outtexturedataarray 1))) _asstatic_mesh_component)))))
